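# UI material for the IWD takeover Item Shop background, showing an animated purple sky.
# Two panned noise textures are summed, sharpened by a power and combined with vertical UV gradients into moving masks.
# These masks and inverted sky luminance blend the sky texture into a purple vertical gradient, then multiply in a dark-purple tint.
# A separate radial-gradient smoke branch uses rerouted UI UVs and does not feed the sky chain.
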Material /ItemShop/UI/Art/IWDTakeover/M_UI_IWD_StoreBackground_2
graph {
 properties {bUsedWithStaticMesh=false, materialDomain=MD_UI, pixelDepthOffsetMode=PDOM_Legacy, usedShadingModels=Default Lit}
 nodes [
  n1 = ScalarParameter() {defaultValue=-0.95, parameterName=Smoke-Distortion-Intensity}
  n2 = MaterialFunctionCall(
   UVs=Add(A=NamedRerouteUsage(), B=Multiply(B=@n1))
   CenterPosition=AppendVector(A=ScalarParameter() {defaultValue=0.5, parameterName=Smoke-OffsetX}, B=ScalarParameter() {defaultValue=0.5, parameterName=Smoke-OffsetY})
   Radius=ScalarParameter() {defaultValue=1, group=Background, parameterName=Radius-Smoke, sortPriority=1}
   Density=ScalarParameter() {defaultValue=1.906, group=Background, parameterName=Density-Smoke, sortPriority=1}
  ) {materialFunction=&RadialGradientExponential}
  Multiply(
   A=Multiply(A=@n2.RadialGradientExponential, B=LinearInterpolate(Alpha=@n2.RadialGradientExponential))
   B=ScalarParameter() {defaultValue=1.7981, parameterName=Smoke-Intensity}
  )
  n3 = TextureSampleParameter2D() {parameterName=Param, texture=/ItemShop/UI/Art/IWDTakeover/Texture/T_UI_Sky}
  n4 = TextureCoordinate()
  n5 = LinearInterpolate(
   A=Constant3Vector() {constant=<rgba 0.0585 0.0064 0.0938 0>}
   B=Constant3Vector() {constant=<rgba 0.1942 0.0136 0.9062 0>}
   Alpha=OneMinus(None=ComponentMask(None=@n4) {g=true})
  )
  n6 = ScalarParameter() {defaultValue=0.9976}
  n7 = Multiply(A=Time(), B=ScalarParameter() {defaultValue=0.1, parameterName=Time Factor})
  n8 = Panner(Time=@n7) {speedX=0.35, speedY=0.125}
  n9 = Panner(Time=@n7) {speedX=0.25, speedY=0.25}
  n10 = Power(
   Base=Add(
    A=TextureSample(UVs=@n8) {samplerSource=SSM_Wrap_WorldGroupSettings, texture=/Game/UI/Frontend/ItemShop/Art/IWDTakeover/Texture/T-Noise-IWD}.G
    B=TextureSample(UVs=@n9) {samplerType=SAMPLERTYPE_Grayscale, texture=&LowResBlurredNoise}.B
   )
   Exp=ScalarParameter() {defaultValue=2.5325, parameterName=Param_1}
  )
  n11 = Add(A=TextureCoordinate(), B=Constant2Vector() {g=-0.1156}) {constB=0.5}
  n12 = Saturate(None=Add(A=@n10, B=ComponentMask(None=@n11) {g=true}))
  n13 = Power(Base=OneMinus(None=Desaturation(None=@n3.RGB)), Exp=ScalarParameter() {defaultValue=0.9976})
  n14 = ComponentMask(None=Add(A=TextureCoordinate(), B=Constant2Vector() {g=-0.198}) {constB=0.5}) {g=true}
  n15 = Multiply(A=Multiply(A=OneMinus(None=@n14)), B=@n10) {constB=2}
  Multiply(
   A=LinearInterpolate(A=@n3.RGB, B=Power(Base=@n5, Exp=@n6), Alpha=LinearInterpolate(B=@n12, Alpha=@n13))
   B=LinearInterpolate(B=Constant3Vector() {constant=<rgba 0.0436 0 0.1667 0>}, Alpha=Saturate(None=@n15)) {constA=1}
  ) {constA=1}
  n16 = MaterialFunctionCall() {materialFunction=&GetUserInterfaceUV}
  n17 = MaterialFunctionCall(Float2=@n16.Pixel Size) {materialFunction=&BreakOutFloat2Components}
  n18 = AppendVector(A=Constant() {r=1}, B=Divide(A=@n17.G, B=@n17.G))
  NamedRerouteDeclaration(None=Multiply(A=Multiply(A=@n18, B=@n16.Normalized UV), B=Constant2Vector() {g=1, r=1})) {name=TextureCoordinate, nodeColor=<rgba 0 0.6953 1 1>}
 ]
}
