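# UI material for the IWD takeover item shop background, showing smoky sky over a purple vertical gradient.
# Two panning noise textures distort radial-gradient UVs to form animated smoke, which multiplies the sky texture along with a vertical falloff band.
# A UV1 vertical mask, adjusted by IsMaskAligned, blends that result over the gradient.
# A second smoke chain, a rotator and a spare noise sample appear unconnected.
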Material /ItemShop/UI/Art/IWDTakeover/M_UI_IWD_StoreBackground_jic
graph {
 properties {bUsedWithStaticMesh=false, materialDomain=MD_UI, pixelDepthOffsetMode=PDOM_Legacy, usedShadingModels=Default Lit}
 nodes [
  n1 = ComponentMask(None=TextureCoordinate()) {g=true}
  n2 = ConstantBiasScale(None=TextureCoordinate()) {bias=-0.5, scale=2}
  n3 = Subtract(A=ComponentMask(None=Abs(None=@n2)) {g=true}, B=ScalarParameter() {defaultValue=1.0273, parameterName=Subtract})
  n4 = ScalarParameter() {defaultValue=-0.4714, parameterName=Falloff}
  n5 = Multiply(A=Saturate(None=Multiply(A=@n3, B=@n4)), B=ScalarParameter() {defaultValue=6.1315, parameterName=Intensity})
  n6 = Multiply(A=Time(), B=ScalarParameter() {defaultValue=0.05, parameterName=Time Factor})
  n7 = Add(
   A=TextureSample(UVs=Panner(Time=@n6) {speedX=0.35, speedY=0.125}) {samplerSource=SSM_Wrap_WorldGroupSettings, texture=/Game/UI/Frontend/ItemShop/Art/IWDTakeover/Texture/T-Noise-IWD}.G
   B=TextureSample(UVs=Panner(Time=@n6) {speedX=0.25, speedY=0.25}) {samplerType=SAMPLERTYPE_Grayscale, texture=&LowResBlurredNoise}.B
  )
  n8 = ScalarParameter() {defaultValue=2.4283, parameterName=Smoke-Distortion-Intensity}
  n9 = MaterialFunctionCall(
   UVs=Add(A=NamedRerouteUsage(), B=Multiply(A=@n7, B=@n8))
   CenterPosition=AppendVector(A=ScalarParameter() {defaultValue=0.5, parameterName=Smoke-OffsetX}, B=ScalarParameter() {defaultValue=0.5, parameterName=Smoke-OffsetY})
   Radius=ScalarParameter() {defaultValue=4.1664, group=Background, parameterName=Radius-Smoke, sortPriority=1}
   Density=ScalarParameter() {defaultValue=2.1529, group=Background, parameterName=Density-Smoke, sortPriority=1}
  ) {materialFunction=&RadialGradientExponential}
  n10 = Multiply(
   A=Multiply(A=@n9.RadialGradientExponential, B=LinearInterpolate(Alpha=@n9.RadialGradientExponential))
   B=ScalarParameter() {defaultValue=2.8371, parameterName=Smoke-Intensity}
  )
  n11 = TextureCoordinate() {coordinateIndex=1, uTiling=0, vTiling=0.648}
  n12 = ComponentMask(None=TextureCoordinate() {coordinateIndex=1, uTiling=0.832}) {g=true}
  LinearInterpolate(
   A=LinearInterpolate(
    A=Constant3Vector() {constant=<rgba 0.15 0.0742 0.3278 0>}
    B=Constant3Vector() {constant=<rgba 0.0612 0.0356 0.1912 0>}
    Alpha=OneMinus(None=@n1)
   )
   B=Multiply(A=TextureSampleParameter2D() {parameterName=Param, texture=/ItemShop/UI/Art/IWDTakeover/Texture/T_UI_Sky}.RGB, B=Multiply(A=@n5, B=@n10))
   Alpha=LinearInterpolate(A=ComponentMask(None=@n11) {g=true}, B=OneMinus(None=@n12), Alpha=ScalarParameter() {defaultValue=0.6976, parameterName=IsMaskAligned})
  ) {constA=1, constAlpha=1}
  n13 = ScalarParameter() {defaultValue=2.4283, parameterName=Smoke-Distortion-Intensity}
  n14 = MaterialFunctionCall(
   UVs=Add(A=NamedRerouteUsage(), B=Multiply(B=@n13))
   CenterPosition=AppendVector(A=ScalarParameter() {defaultValue=2.9, parameterName=Smoke-Offset}, B=ScalarParameter() {defaultValue=2.9, parameterName=Smoke-Offset})
   Radius=ScalarParameter() {defaultValue=4.1664, group=Background, parameterName=Radius-Smoke, sortPriority=1}
   Density=ScalarParameter() {defaultValue=2.1529, group=Background, parameterName=Density-Smoke, sortPriority=1}
  ) {materialFunction=&RadialGradientExponential}
  Multiply(
   A=Multiply(A=@n14.RadialGradientExponential, B=LinearInterpolate(Alpha=@n14.RadialGradientExponential))
   B=ScalarParameter() {defaultValue=2.8371, parameterName=Smoke-Intensity}
  )
  n15 = MaterialFunctionCall() {materialFunction=&GetUserInterfaceUV}
  n16 = MaterialFunctionCall(Float2=@n15.Pixel Size) {materialFunction=&BreakOutFloat2Components}
  n17 = AppendVector(A=Constant() {r=1}, B=Divide(A=@n16.G, B=@n16.G))
  NamedRerouteDeclaration(None=Multiply(A=Multiply(A=@n17, B=@n15.Normalized UV), B=Constant2Vector() {g=2, r=1})) {name=TextureCoordinate, nodeColor=<rgba 0 0.6953 1 1>}
  Rotator() {speed=0.02}
  TextureSample() {samplerType=SAMPLERTYPE_Grayscale, texture=&LowResBlurredNoise}
 ]
}
